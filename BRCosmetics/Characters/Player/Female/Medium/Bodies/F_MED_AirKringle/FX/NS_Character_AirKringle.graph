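# Niagara snow effect for the AirKringle character body cosmetic, with ambient flakes and reactive snow bursts.
# Idle flakes spawn around skeleton bones, respect hidden hand/foot flags, and stop while skydiving in gameplay.
# System.ReactivitySpawn gates reactive flakes and falling snowballs, which collide with the world, then change mesh, scale and surface-aligned orientation.
# Trailing flakes spawn from snowball particles through an attribute reader; time of day adjusts flake color and snowball opacity.
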NiagaraSystem /BRCosmetics/Characters/Player/Female/Medium/Bodies/F_MED_AirKringle/FX/NS_Character_AirKringle
dependencies {
 usedDataInterfaces [
  &NiagaraDataInterfaceArrayInt32
  &NiagaraDataInterfaceSpriteRendererInfo
  &NiagaraDataInterfaceCurve
  &NiagaraDataInterfaceColorCurve
  &NiagaraDataInterfaceMeshRendererInfo
  &NiagaraDataInterfaceParticleRead
  &NiagaraDataInterfaceSkeletalMesh
 ]
 usedDynamicInputs [
  &Power
  &Subtract_Float
  &ClampFloat
  &Divide_Float
  &FNTimeOfDayColor
 ]
 usedModules [
  &self:SystemSpawnScript.SystemScriptSource.SystemScriptGraph.NiagaraNodeAssignment_2.SetVariables_#1
  &self:SystemSpawnScript.SystemScriptSource.SystemScriptGraph.NiagaraNodeAssignment_3.SetVariables_#2
  &SystemState
  &self:SystemSpawnScript.SystemScriptSource.SystemScriptGraph.NiagaraNodeAssignment_1.SetVariables_#3
  &self:SystemSpawnScript.SystemScriptSource.SystemScriptGraph.NiagaraNodeAssignment_0.SetVariables_#4
  &self:SystemSpawnScript.SystemScriptSource.SystemScriptGraph.NiagaraNodeAssignment_4.SetVariables_#5
 ]
}
emitters {
 Following_Snowflakes {
  emitterUpdateScript [
   &EmitterState()
   &SpawnParticlesFromOtherEmitter(
    Attribute Reader=DataInterface() {propertyValues {EmitterBinding {bindingMode=Other, emitterName=Reactivity_SnowBalls}}}
    Maximum Age Spawn Threshold=0.3
    Spawn Rate=25
    Use Maximum Age=false
    Use Minimum Age=false
   )
  ]
  particleSpawnScript [
   &InitializeParticle(
    Color=<rgba 0.9 0.9667 1 1>
    Lifetime=&RandomRangeFloat(Minimum=0.4, Maximum=0.6)
    UsePositionOffset=false
    Write Material Random=true
   )
   &SampleParticlesFromOtherEmitter(
    Applied Velocity Scale:NiagaraFloat=&RandomRangeFloat(Minimum=0.1, Maximum=0.2)
    Velocity Application:ENiagara_DirectReadApplicationMode=::NewEnumerator0
    Velocity Sampling=::NewEnumerator0
   )
   &ShapeLocation(
    MaskSpawnGroup=false
    Override Seed=false
    Shape Origin=$Output.SampleParticlesFromOtherEmitter.SampledPosition
    Sphere Radius=&Multiply_Float(A=6, B=&VectorLength(Vector=$Output.SampleParticlesFromOtherEmitter.SampledScale))
    UseOwnerScale=true
   )
   /SetVariables(Particles.SpriteSize=&Multiply_Vector2D_ByFloat(
     Float=&RandomRangeFloat(Minimum=1.5, Maximum=3)
     Vector2D=&Vector2DFromFloat(Value=&VectorLength(Vector=$Output.SampleParticlesFromOtherEmitter.SampledScale))
    )) {isSetParameters=true}
  ]
  particleUpdateScript [
   &ParticleState()
   &GravityForce(Gravity=&MakeVector(X=0, Y=0, Z=&RandomRangeFloat(Minimum=-10, Maximum=-20)))
   &CurlNoiseForce(Bias Noise=false, Noise Strength=25, Pan Noise=false)
   &SolveForcesAndVelocity(Clamp Velocity=false, Limit Acceleration=false)
   &ScaleSpriteSizeBySpeed(Max Scale Factor=<1 2>, Velocity Threshold=300)
   &ScaleColor(
    Scale Alpha=&FloatFromCurve(FloatCurve=DataInterface() {propertyValues {Curve {keys [
         {interpMode=RCIM_Cubic, tangentMode=RCTM_User, time=0, value=0}
         {interpMode=RCIM_Cubic, tangentMode=RCTM_User, time=0.1, value=1}
         {interpMode=RCIM_Cubic, tangentMode=RCTM_User, time=1, value=0}
        ]}}})
    Scale RGB=&MakeVectorFromLinearColorRGB(Color=&MultiplyLinearColor(
      Linear Color=&ColorFromCurve(ColorCurve=DataInterface() {propertyValues {
         AlphaCurve {keys []}
         BlueCurve {keys [<at 0.2426 1 RCIM_Cubic>,<at 1 1 RCIM_Cubic>]}
         GreenCurve {keys [<at 0.2426 1 RCIM_Cubic>,<at 1 0.755 RCIM_Cubic>]}
         RedCurve {keys [<at 0.2426 1 RCIM_Cubic>,<at 1 0.65 RCIM_Cubic>]}
        }})
      Scale Factor=&MakeVector4_FromLinearColor(LinearColor=$System.TOD Color)
     ))
    ScaleA=true
    ScaleRGB=true
   )
  ]
  rendererClasses [&NiagaraSpriteRendererProperties]
  renderers [{rendererClass=&NiagaraSpriteRendererProperties}]
 }
 Idle_SnowFlakes {
  emitterUpdateScript [
   &EmitterState()
   &SpawnRate(
    SpawnRate=&FN_FrontEndFloatModifier(Gameplay=&MakeCustomFloatFromBool(Bool=$User.IsSkydiving, True Float=0, False Float=10), Front End=10)
    Use Spawn Probability=false
   )
  ]
  particleSpawnScript [
   &InitializeParticle(
    Color=<rgba 0.9 0.9667 1 0.1>
    Lifetime Max=1.5
    Lifetime Mode=::NewEnumerator1
    Sprite Rotation Mode=::NewEnumerator2
    Sprite Size Mode=::NewEnumerator1
    Uniform Sprite Size Max=5
    Uniform Sprite Size Min=3
    UsePositionOffset=false
    Write Material Random=true
   )
   &SkeletalMeshLocation(
    Bone / Socket Index=&MakeCustomIntFromBool(
     Bool=$User.bisHandHidden
     True INT=&MakeCustomIntFromBool(
      Bool=$User.bisFootHidden
      True INT=8
      False INT=&SelectIntFromArray(
       Array Sampling Mode=::NewEnumerator1
       Direct Array Index=?
       Int Selection Array=DataInterface() {propertyValues {IntData [0,1,2,3,8]}}
       Random Seed=0
      )
     )
     False INT=&MakeCustomIntFromBool(
      Bool=$User.bisFootHidden
      True INT=&RandomRangeInt(Minimum=4, Maximum=8)
      False INT=&RandomRangeInt(Minimum=0, Maximum=8)
     )
    )
    Bone Sampling Mode=::NewEnumerator4
    Inherit Velocity=false
    MaskSpawnGroup=false
    Skeletal Mesh=$System.Skeletal Mesh
   )
   &ShapeLocation(MaskSpawnGroup=false, Override Seed=false, Shape Origin=$Output.SkeletalMeshLocation.SampledPosition, Sphere Radius=12, UseOwnerScale=true)
   &AddVelocity(
    Module.Use Falloff Distance=false
    Override Seed=false
    Velocity Mode=::NewEnumerator1
    Velocity Origin=$Output.SkeletalMeshLocation.SampledPosition
    Velocity Speed=&RandomRangeFloat(Minimum=30, Maximum=40)
   ) {module=AddVelocity001}
   /SetVariables(Particles.RandFacing:Vector3f=&RandomRangeVector(Minimum=<-1 -1 -1>, Maximum=<1 1 1>)) {isSetParameters=true}
   &CameraOffset(Camera Offset Amount=6)
   &SubUVAnimation(
    End Frame Range Override=3
    Evaluation Type=::NewEnumerator0
    "Sprite Renderer "=DataInterface() {propertyValues {SpriteRenderer=&self:Idle_SnowFlakes_2.NiagaraSpriteRendererProperties_2}}
    SubUV Animation Mode:ENiagara_SubUVLookupModeV2=::NewEnumerator2
    UseEndFrame=true
    UseStartFrame=true
   )
  ]
  particleUpdateScript [
   &ParticleState()
   &SpriteRotationRate(Rotation Rate=&RandomRangeFloat(Minimum=-100, Maximum=100))
   &CurlNoiseForce(Noise Frequency=22, Noise Strength=50, Pan Noise=true, Pan Noise Field=<0.3 0.3 0.3>)
   &AccelerationForce(Acceleration=&RandomRangeVector(Minimum=<0 0 -10>, Maximum=<0 0 -50>))
   &Drag(Drag=&RandomRangeFloat(Minimum=2, Maximum=5), Use Linear Drag=true, Use Rotational Drag=false)
   &SolveForcesAndVelocity(Clamp Velocity=false, Limit Acceleration=false)
   &ScaleColor(
    Scale Alpha=&FloatFromCurve(FloatCurve=DataInterface() {propertyValues {Curve {keys [
         {interpMode=RCIM_Cubic, tangentMode=RCTM_User, time=0, value=0}
         {interpMode=RCIM_Cubic, tangentMode=RCTM_User, time=0.1, value=1}
         {interpMode=RCIM_Cubic, tangentMode=RCTM_User, time=1, value=0}
        ]}}})
    Scale RGB=&MakeVectorFromLinearColorRGB(Color=&MultiplyLinearColor(
      Linear Color=&ColorFromCurve(ColorCurve=DataInterface() {propertyValues {
         AlphaCurve {keys []}
         BlueCurve {keys [<at 0.2426 1 RCIM_Cubic>,<at 1 1 RCIM_Cubic>]}
         GreenCurve {keys [<at 0.2426 1 RCIM_Cubic>,<at 1 0.755 RCIM_Cubic>]}
         RedCurve {keys [<at 0.2426 1 RCIM_Cubic>,<at 1 0.65 RCIM_Cubic>]}
        }})
      Scale Factor=&MakeVector4_FromLinearColor(LinearColor=$System.TOD Color)
     ))
    ScaleA=true
    ScaleRGB=true
   )
   &ScaleSpriteSize(Scale Factor=&Vector2DFromFloat(Value=&FloatFromCurve(FloatCurve=DataInterface() {propertyValues {Curve {keys [
          <at 0 0 RCIM_Cubic>
          {arriveTangent=2, interpMode=RCIM_Cubic, leaveTangent=2, time=0.1619, value=1}
          {arriveTangent=-1.1932, interpMode=RCIM_Cubic, leaveTangent=-1.1932, time=0.5, value=1}
          <at 1 0 RCIM_Cubic>
         ]}}})))
  ]
  rendererClasses [&NiagaraSpriteRendererProperties]
  renderers [{rendererClass=&NiagaraSpriteRendererProperties}]
 }
 Reactivity_SnowBalls {
  emitterSpawnScript [/SetVariables(Emitter.RandomSeed:NiagaraInt32=/Niagara/DynamicInputs/Random/RandomInt(Randomness Mode:ENiagaraRandomnessMode=::NewEnumerator0, Random Seed:NiagaraInt32=0, Fixed Random Seed:NiagaraBool=false)) {isSetParameters=true}]
  emitterUpdateScript [
   &EmitterState()
   &SpawnRate(Spawn Probability=$System.ReactivitySpawn, SpawnRate=35, Use Spawn Probability=true)
  ]
  particleSpawnScript [
   &InitializeParticle(
    Color=<rgba 0.6 1.35 1.5 0.3>
    Lifetime Max=1.1
    Lifetime Min=0.8
    Lifetime Mode=::NewEnumerator1
    Mesh Scale Mode=::NewEnumerator1
    Mesh Uniform Scale Max=1
    Mesh Uniform Scale Min=0.7
    Sprite Size Mode=::NewEnumerator1
    Uniform Sprite Size Max=80
    Uniform Sprite Size Min=60
    UsePositionOffset=false
    Write Material Random=true
   )
   &SkeletalMeshLocation(
    Bone / Socket Index=8
    Bone Sampling Mode=::NewEnumerator4
    Inherit Velocity=false
    MaskSpawnGroup=false
    Skeletal Mesh=$System.Skeletal Mesh
   )
   &InheritVelocity(Inherited Velocity Amount Scale=<0.6 0.6 0.3>, Inherited Velocity Speed Limit=500, Speed Limit=true, UseSpeedThreshold=false)
   &ShapeLocation(
    Cylinder Distribution=::NewEnumerator2
    Cylinder Height=30
    Cylinder Height Midpoint=0.8
    Cylinder Radius=30
    MaskSpawnGroup=false
    Shape Origin=$Output.SkeletalMeshLocation.SampledPosition
    Shape Primitive=::NewEnumerator1
    U Position=&RandomRangeFloat(Minimum=0, Maximum=1)
    UseOwnerScale=true
    V Position=&RandomRangeFloat(Minimum=0, Maximum=1)
   )
   &AddVelocity(
    Module.Use Falloff Distance=false
    Origin Offset=<0 0 70>
    Override Seed=false
    Velocity Mode=::NewEnumerator1
    Velocity Speed=&RandomRangeFloat(Minimum=50, Maximum=100)
   ) {module=AddVelocity002}
   &InitialMeshOrientation(
    Mesh Orientation Mode=::NewEnumerator2
    Orientation Vector=<0 0 -1>
    Rotation=&RandomRangeVector(Minimum=<0 0 -1>, Maximum=<0 0 1>)
    Use Rotation Vector=true
   )
   /SetVariables(Particles.SpriteCameraOffset=20) {isSetParameters=true}
  ]
  particleUpdateScript [
   &ParticleState()
   &GravityForce(Gravity=&MakeVector(X=0, Y=0, Z=&RandomRangeFloat(Minimum=-800, Maximum=-1200)))
   &Collision(
    Age Colliding Particles=false
    CPU Collision Trace Channel=ECC_WorldStatic
    Collision Enabled=&SetBoolByFloatComparison(
     A=&MakeFloatFromVector(Vector=Linked() {linkedVariable {name=Particles.Position, type=NiagaraPosition}}, Channel=::NewEnumerator2)
     B=&Add_Float(
      A=&MakeFloatFromVector(Vector=Linked() {linkedVariable {name=Engine.Owner.Position, type=NiagaraPosition}}, Channel=::NewEnumerator2)
      B=25
     )
     Comparison Type=::NewEnumerator6
    )
    Friction:NiagaraFloat=&FloatFromCurve(
     FloatCurve=DataInterface() {propertyValues {Curve {keys [
         {interpMode=RCIM_Cubic, tangentMode=RCTM_User, time=0, value=0}
         {arriveTangent=0.2634, interpMode=RCIM_Cubic, leaveTangent=0.2634, tangentMode=RCTM_User, time=0.6, value=1}
        ]}}}
     Scale Curve=30
    )
    KillOnCollision=false
    Max Number Of Collisions=3
    Maximum Penetration Correction Distance=0
    Mesh Dimensions:Vector3f=<9 8 5>
    Module.Control Rest State Via Surface Normal=false
    Module.EnableMaxCollisionCount=true
    Module.Randomize Collision Normal Vector=false
    Percentage of Penetration Before Rest=0
    Radius Calculation Type=::NewEnumerator1
    Rest State Time Range=0
    Restitution=0
   )
   &Drag(Drag=0, Use Linear Drag=true, Use Rotational Drag=true)
   &SolveForcesAndVelocity(Clamp Velocity=false, Limit Acceleration=false)
   /SetVariables(Particles.VisibilityTag=&MakeCustomIntFromBool(Bool=$Particles.HasCollided, True INT=1, False INT=0)) {isSetParameters=true}
   /SetVariables(Particles.MeshOrientation=/Niagara/DynamicInputs/TypeConversions/MakeCustomQuaternionFromBool(
     Bool:NiagaraBool=$Particles.HasCollided
     True Quat:Quat4f=&OrientMeshToVector(Look At Direction=$Output.Collision.CombinedCollisionNormal, Reference Axis Vector=<1 0 0>)
     False Quat:Quat4f=&MakeQuatFromAxisAngle(Axis=<1 0 0>, AngleInDegrees=&RandomRangeFloat(Minimum=-360, Maximum=360))
    )) {isSetParameters=true}
   &ScaleMeshSize(Scale Factor=&MakeCustomVectorFromBool(
     Bool=$Particles.HasCollided
     True Vector=&RandomRangeVector(Minimum=<1 1 1.8>, Maximum=<1.4 1.4 3>)
     False Vector=&RandomRangeVector(Minimum=<0.8 1 1>, Maximum=<1 2 2>)
    )) {module=ScaleMeshSize001}
   &ScaleMeshSize(Scale Factor=&VectorFromFloat(Value=&FloatFromCurve(FloatCurve=DataInterface() {propertyValues {Curve {keys [
          {arriveTangent=3.5292, interpMode=RCIM_Cubic, leaveTangent=3.5292, tangentMode=RCTM_User, time=0, value=0.4}
          {interpMode=RCIM_Cubic, tangentMode=RCTM_User, time=0.2, value=1}
          {interpMode=RCIM_Cubic, tangentMode=RCTM_User, time=0.85, value=1}
          {arriveTangent=-5.6776, interpMode=RCIM_Cubic, leaveTangent=-5.6776, tangentMode=RCTM_User, time=1, value=0}
         ]}}})))
   &ScaleMeshSizeBySpeed(Max Scale Factor=&RandomRangeVector(Minimum=<2 1.2 1.2>, Maximum=<3 1.5 1.5>), Speed Threshold=570)
   /SetVariables(Particles.MeshIndex=&MakeCustomIntFromBool(
     Bool=$Particles.HasCollided
     True INT=&RandomRangeInt(Minimum=0, Maximum=1)
     False INT=&RandomRangeInt(Minimum=0, Maximum=3)
    )) {isSetParameters=true}
   &ScaleColor(
    Scale Alpha=&FloatFromCurve(
     FloatCurve=DataInterface() {propertyValues {Curve {keys [{interpMode=RCIM_Cubic, tangentMode=RCTM_User, time=0, value=1}, {interpMode=RCIM_Cubic, tangentMode=RCTM_User, time=0.9, value=0}]}}}
     Scale Curve=&FNTimeOfDayScalar(Night Scale=0.5, Morning Scale=0.6, Day Scale=0.3, Evening Scale=0.5)
    )
    ScaleA=true
    ScaleRGB=true
   )
   &ScaleSpriteSize(
    Uniform Curve Scale=&FN_FrontEndFloatModifier(Gameplay=1, Front End=1)
    Uniform Curve Sprite Scale=DataInterface() {propertyValues {Curve {keys [
        {arriveTangent=0.0315, interpMode=RCIM_Cubic, leaveTangent=0.0315, tangentMode=RCTM_User, time=0, value=0}
        {interpMode=RCIM_Cubic, tangentMode=RCTM_User, time=1, value=1}
       ]}}}
   )
   &SpriteRotationRate(Rotation Rate=&RandomRangeFloat(Minimum=-250, Maximum=250))
   &DynamicMaterialParameters(
    Index 0 Param 1=&Add_Float(
     A=&RandomRangeFloat(Minimum=0, Maximum=1)
     B=&FloatFromCurve(
      FloatCurve=DataInterface() {propertyValues {Curve {keys [
          <at 0 0 RCIM_Cubic>
          {arriveTangent=1.5947, interpMode=RCIM_Cubic, leaveTangent=1.5947, tangentMode=RCTM_User, time=0.2557, value=0.5257}
          <at 1 1 RCIM_Cubic>
         ]}}}
      CurveIndex=$Particles.Age
      Scale Curve=0.25
     )
    )
    Module.Param0WriteEnabled=true
    Module.Param1WriteEnabled=true
    Module.Param2WriteEnabled=true
    Module.Param3WriteEnabled=true
   ) {module=DynamicMaterialParameters001}
  ]
  rendererClasses [&NiagaraMeshRendererProperties, &NiagaraSpriteRendererProperties]
  renderers [
   {rendererClass=&NiagaraMeshRendererProperties, rendererIndex=0}
   {rendererClass=&NiagaraMeshRendererProperties, rendererIndex=1}
   {rendererClass=&NiagaraSpriteRendererProperties, rendererIndex=2}
   {rendererClass=&NiagaraSpriteRendererProperties, rendererIndex=3}
  ]
 }
 Reactivity_SnowFlakes {
  emitterUpdateScript [
   &EmitterState(Life Cycle Mode=::NewEnumerator1, Loop Duration=1, UseLoopDelay=false)
   &SpawnRate(Spawn Probability=$System.ReactivitySpawn, SpawnRate=140, Use Spawn Probability=true)
  ]
  particleSpawnScript [
   &InitializeParticle(
    Color=<rgba 0.9 0.9667 1 0.1>
    Lifetime Max=1
    Lifetime Min=0.7
    Lifetime Mode=::NewEnumerator1
    Sprite Size Mode=::NewEnumerator1
    Uniform Sprite Size Max=5
    Uniform Sprite Size Min=3
    UsePositionOffset=false
    Write Material Random=true
   )
   &SkeletalMeshLocation(
    Bone / Socket Index=&MakeCustomIntFromBool(
     Bool=$User.bisHandHidden
     True INT=&MakeCustomIntFromBool(
      Bool=$User.bisFootHidden
      True INT=8
      False INT=&SelectIntFromArray(
       Array Sampling Mode=::NewEnumerator1
       Direct Array Index=?
       Int Selection Array=DataInterface() {propertyValues {IntData [0,1,2,3,8]}}
       Random Seed=0
      )
     )
     False INT=&MakeCustomIntFromBool(
      Bool=$User.bisFootHidden
      True INT=&RandomRangeInt(Minimum=4, Maximum=8)
      False INT=&RandomRangeInt(Minimum=0, Maximum=8)
     )
    )
    Bone Sampling Mode=::NewEnumerator4
    Inherit Velocity=false
    MaskSpawnGroup=false
    Skeletal Mesh=$System.Skeletal Mesh
   )
   &ShapeLocation(MaskSpawnGroup=false, Override Seed=false, Shape Origin=$Output.SkeletalMeshLocation.SampledPosition, Sphere Radius=15, UseOwnerScale=true)
   &AddVelocity(
    Module.Use Falloff Distance=false
    Origin Offset=<0 0 70>
    Override Seed=false
    Velocity Mode=::NewEnumerator1
    Velocity Speed=&RandomRangeFloat(Minimum=50, Maximum=80)
   ) {module=AddVelocity002}
   &CameraOffset(Camera Offset Amount=6)
   &SubUVAnimation(
    End Frame Range Override=3
    Evaluation Type=::NewEnumerator0
    "Sprite Renderer "=DataInterface() {propertyValues {SpriteRenderer=&self:Reactivity_SnowFlakes_2.NiagaraSpriteRendererProperties_1}}
    SubUV Animation Mode:ENiagara_SubUVLookupModeV2=::NewEnumerator2
    UseEndFrame=true
    UseStartFrame=true
   )
  ]
  particleUpdateScript [
   &ParticleState()
   &CurlNoiseForce(
    Noise Frequency=200
    Noise Strength=&FloatFromCurve(
     FloatCurve=DataInterface() {propertyValues {Curve {keys [{interpMode=RCIM_Cubic, tangentMode=RCTM_User, time=0.2, value=2}, {interpMode=RCIM_Cubic, tangentMode=RCTM_User, time=0.45, value=0.5}]}}}
     Scale Curve=100
    )
    Pan Noise=true
    Pan Noise Field=<0.3 0.3 0.3>
   )
   &AccelerationForce(Acceleration=<0 0 -150>)
   &Drag(Drag=&RandomRangeFloat(Minimum=2, Maximum=4), Use Linear Drag=true, Use Rotational Drag=false)
   &SolveForcesAndVelocity(Clamp Velocity=false, Limit Acceleration=false)
   &ScaleColor(
    Scale Alpha=&FloatFromCurve(FloatCurve=DataInterface() {propertyValues {Curve {keys [
         {interpMode=RCIM_Cubic, tangentMode=RCTM_User, time=0, value=0}
         {interpMode=RCIM_Cubic, tangentMode=RCTM_User, time=0.1, value=1}
         {interpMode=RCIM_Cubic, tangentMode=RCTM_User, time=1, value=0}
        ]}}})
    Scale RGB=&MakeVectorFromLinearColorRGB(Color=&MultiplyLinearColor(
      Linear Color=&ColorFromCurve(ColorCurve=DataInterface() {propertyValues {
         AlphaCurve {keys []}
         BlueCurve {keys [<at 0.2426 1 RCIM_Cubic>,<at 1 1 RCIM_Cubic>]}
         GreenCurve {keys [<at 0.2426 1 RCIM_Cubic>,<at 1 0.755 RCIM_Cubic>]}
         RedCurve {keys [<at 0.2426 1 RCIM_Cubic>,<at 1 0.65 RCIM_Cubic>]}
        }})
      Scale Factor=&MakeVector4_FromLinearColor(LinearColor=$System.TOD Color)
     ))
    ScaleA=true
    ScaleRGB=true
   )
   &ScaleSpriteSize(Scale Factor=&Vector2DFromFloat(Value=&FloatFromCurve(FloatCurve=DataInterface() {propertyValues {Curve {keys [
          <at 0 0 RCIM_Cubic>
          {arriveTangent=2, interpMode=RCIM_Cubic, leaveTangent=2, time=0.1619, value=1}
          {arriveTangent=-1.1932, interpMode=RCIM_Cubic, leaveTangent=-1.1932, time=0.5, value=1}
          <at 1 0 RCIM_Cubic>
         ]}}})))
  ]
  rendererClasses [&NiagaraSpriteRendererProperties]
  renderers [{rendererClass=&NiagaraSpriteRendererProperties}]
 }
}
summary {
 systemName=NS_Character_AirKringle
 emitters [
  {emitterName=Reactivity_SnowFlakes, rendererClasses [&NiagaraSpriteRendererProperties]}
  {emitterName=Idle_SnowFlakes, rendererClasses [&NiagaraSpriteRendererProperties]}
  {
   emitterName=Reactivity_SnowBalls
   rendererClasses [&NiagaraMeshRendererProperties, &NiagaraSpriteRendererProperties]
  }
  {emitterName=Following_Snowflakes, rendererClasses [&NiagaraSpriteRendererProperties]}
 ]
 userVariables [
  User.bisFootHidden {type=NiagaraBool}
  User.bIsFrontEnd {type=NiagaraBool}
  User.bIsFrontEndPreview {type=NiagaraBool}
  User.bisHandHidden {type=NiagaraBool}
  User.IsSkydiving {type=NiagaraBool}
  User.JumpGameTime {type=NiagaraFloat, defaultValue=0, description=""}
  User.LandGameTime {type=NiagaraFloat, defaultValue=0, description=""}
  User.ReactivityDebug {type=NiagaraBool}
 ]
}
