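# Physics asset for the F_MED_AirKringle face accessory mesh, simulating a dangling scarf on a female player character.
# Three scarf bodies built from capsules chain from a base sphere, with narrow swing limits and very tight twist to restrain bending.
# The scarf base and a button sphere are angularly locked to spine_04.
# Large capsules on three spine bones, joined by angularly free constraints, approximate the torso.
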PhysicsAsset /BRCosmetics/Characters/Player/Female/Medium/Bodies/F_MED_AirKringle/Meshes/Parts/F_MED_AirKringle_FaceAcc_Physics
bodies {
 button [{shapeType=Sphere, radius=2.6091, shapeName=button_sphere}]
 scarf_01_base [{shapeType=Sphere, radius=1, shapeName=scarf_MSTR_sphere}]
 scarf_01_dyn [{shapeType=Capsule, length=5.5, radius=3, shapeName=dyn_scarf_01_capsule, center=<3.1271 0 -1.4262>, rotation=<rot 90 180 -99.0341>}]
 scarf_02_dyn [
  {shapeType=Capsule, length=7.579, radius=3, shapeName=dyn_scarf_02_capsule, center=<3.1266 -0.5412 -1.9378>, rotation=<rot 0 179.9472 90>}
  {
   shapeType=Capsule
   length=6.7025
   radius=3.5445
   shapeName=scarf_02_dyn_capsule
   center=<0.0604 -6.0829 -1.8728>
   rotation=<rot 88.0364 160.2484 179.9988>
  }
 ]
 scarf_03_dyn [
  {
   shapeType=Capsule
   length=11.2504
   radius=4
   shapeName=dyn_scarf_04_capsule
   center=<8.2968 -0.8895 -2.269>
   rotation=<rot 81.1443 -139.873 -60.2395>
  }
  {
   shapeType=Capsule
   length=11.285
   radius=2.0499
   shapeName=scarf_03_dyn_capsule
   center=<-0.3522 -6.407 0.3652>
   rotation=<rot 13.328 -13.2341 -82.6549>
  }
 ]
 spine_02 [{shapeType=Capsule, length=0, radius=20, shapeName=spine_02_capsule, center=<0 -1.1119 0>, rotation=<rot 90 0 0>}]
 spine_03 [{shapeType=Capsule, length=0.8023, radius=19.7047, shapeName=spine_03_capsule, center=<0.4882 -1.828 0>, rotation=<rot 90 179.6419 -175.1104>}]
 spine_04 [{shapeType=Capsule, length=14, radius=16, shapeName=spine_04_capsule, center=<0 -3.6399 0.0601>, rotation=<rot 0 0 -10.9435>}]
}
constraints [
 {bone1Name=scarf_01_dyn, bone2Name=scarf_01_base, swing1LimitDegrees=20, swing2LimitDegrees=20, twistLimitDegrees=2}
 {bone1Name=scarf_02_dyn, bone2Name=scarf_01_dyn, swing1LimitDegrees=20, swing2LimitDegrees=25, twistLimitDegrees=1}
 {bone1Name=scarf_03_dyn, bone2Name=scarf_02_dyn, swing1LimitDegrees=20, swing2LimitDegrees=20, twistLimitDegrees=1}
 {
  bone1Name=scarf_01_base
  bone2Name=spine_04
  swing1LimitDegrees=0
  swing1Motion=Locked
  swing2LimitDegrees=20
  swing2Motion=Locked
  twistLimitDegrees=1
  twistMotion=Locked
 }
 {
  bone1Name=spine_03
  bone2Name=spine_02
  swing1LimitDegrees=45
  swing1Motion=Free
  swing2LimitDegrees=45
  swing2Motion=Free
  twistLimitDegrees=45
  twistMotion=Free
 }
 {
  bone1Name=spine_04
  bone2Name=spine_03
  swing1LimitDegrees=45
  swing1Motion=Free
  swing2LimitDegrees=45
  swing2Motion=Free
  twistLimitDegrees=45
  twistMotion=Free
 }
 {
  bone1Name=button
  bone2Name=spine_04
  swing1LimitDegrees=45
  swing1Motion=Locked
  swing2LimitDegrees=45
  swing2Motion=Locked
  twistLimitDegrees=45
  twistMotion=Locked
 }
]
